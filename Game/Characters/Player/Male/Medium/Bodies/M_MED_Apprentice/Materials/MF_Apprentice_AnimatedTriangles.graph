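# Optional material function that adds animated two-layer triangle patterns to the Apprentice player body as emissive color within material attributes.
# Two panned samples of the Fryangles wrap mask, differing in tiling and speed, are offset by a UV gradient or scalar and hard-thresholded with Ceil.
# They layer three colors, darkened by a shading curve defaulting to sky-light N·L.
# A LUT call feeds emissive, blended using the minimum of two masks.
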MaterialFunction /Game/Characters/Player/Male/Medium/Bodies/M_MED_Apprentice/Materials/MF_Apprentice_AnimatedTriangles
graph {nodes [
  FunctionOutput(None=StaticSwitchParameter(True=NamedRerouteUsage(), False=NamedRerouteUsage()) {group=1-Apprentice_AnimatedTriangles, parameterName=UseApprentice_AnimatedTriangles, sortPriority=0})
  n1 = TextureCoordinate() {coordinateIndex=1}
  n2 = Multiply(
   A=OneMinus(None=ComponentMask(None=@n1) {g=true})
   B=ScalarParameter() {defaultValue=1.25, group=1-Apprentice_AnimatedTriangles, parameterName=Gradient Lenght, sliderMax=2}
  ) {constB=1.25}
  n3 = StaticSwitchParameter(
   True=Multiply(A=Saturate(None=@n2)) {constB=-1}
   False=ScalarParameter() {group=1-Apprentice_AnimatedTriangles, parameterName=GradientValue, sliderMax=1, sliderMin=-1}
  ) {defaultValue=true, group=1-Apprentice_AnimatedTriangles, parameterName=UseGradient, sortPriority=1}
  n4 = MaterialFunctionCall(XY=StaticSwitchParameter(True=TextureCoordinate() {coordinateIndex=1}, False=TextureCoordinate()) {defaultValue=true, group=1-Apprentice_AnimatedTriangles, parameterName=UseUV2, sortPriority=1}) {materialFunction=&Swizzle}
  n5 = Multiply(
   A=Add(A=@n4.YX, B=Constant() {r=0.5})
   B=ScalarParameter() {defaultValue=1.5, group=1-Apprentice_AnimatedTriangles, parameterName=Triangles2_Tiling, sliderMax=2}
  ) {constB=1.5}
  n6 = ScalarParameter() {defaultValue=1, group=1-Apprentice_AnimatedTriangles, parameterName=AnimationScale}
  n7 = AppendVector(A=Multiply(A=Constant() {r=0.15}, B=@n6), B=Constant())
  n8 = Add(
   A=@n3
   B=TextureSample(UVs=Panner(Coordinate=@n5, Speed=@n7) {speedX=0.15}) {samplerType=SAMPLERTYPE_Masks, texture=/Game/Weapons/WeaponSkins/Wraps/Textures/Fryangles/T_Wrap_Fryangles_M2}.R
  ) {constB=-0.5}
  n9 = ScalarParameter() {defaultValue=0.15, group=1-Apprentice_AnimatedTriangles, parameterName=RemapWidth, sliderMax=1}
  MaterialFunctionCall!(In=@n8, Ramp Width=@n9) {materialFunction=/Engine/Functions/Engine_MaterialFunctions01/Math/SmoothCeil}
  n10 = Multiply(A=@n4.YX, B=ScalarParameter() {defaultValue=1, group=1-Apprentice_AnimatedTriangles, parameterName=Triangles1_Tiling, sliderMax=2})
  n11 = AppendVector(A=Multiply(A=Constant() {r=0.075}, B=@n6), B=Constant())
  n12 = Add(
   A=@n3
   B=TextureSample(UVs=Panner(Coordinate=@n10, Speed=@n11) {speedX=0.075}) {samplerType=SAMPLERTYPE_Masks, texture=&T_Wrap_Fryangles_M2}.R
  ) {constB=-0.5}
  MaterialFunctionCall(In=@n12, Ramp Width=@n9) {materialFunction=&SmoothCeil}
  NamedRerouteDeclaration(None=FunctionInput() {inputName=MA, inputType=FunctionInput_MaterialAttributes}) {name=MA, nodeColor=<rgba 0.108 1 0.2958 1>}
  n13 = ChannelMaskParameter(None=FunctionInput() {inputName=FX, inputType=FunctionInput_Vector4, sortPriority=3}) {group=1-Apprentice_AnimatedTriangles, maskChannel=Green, parameterName=FX_Channel, sortPriority=2, defaultValue=<rgba 0 1 0 0>}
  NamedRerouteDeclaration(None=@n13.R) {name=FX, nodeColor=<rgba 0 0.276 0.0237 1>}
  n14 = NamedRerouteUsage()
  n15 = MaterialFunctionCall!(In=NamedRerouteUsage(), LUT, LUT Multiplier) {materialFunction=/Game/Packages/Fortress_SharedMaterials/Anime_Master_Material/MF_ApplyLUT}
  n16 = NamedRerouteUsage()
  n17 = NamedRerouteUsage()
  NamedRerouteDeclaration(None=BlendMaterialAttributes(
    A=NamedRerouteUsage()
    B=SetMaterialAttributes!(MaterialAttributes=@n14, EmissiveColor=@n15.Result) {attributeSetTypes [#1]}
    Alpha=Min(A=@n16, B=@n17)
   )) {name=Out, nodeColor=<rgba 1 0.4612 0.145 1>}
  n18 = VectorParameter() {group=1-Apprentice_AnimatedTriangles, parameterName=ColorBackground, sortPriority=5, defaultValue=<rgba 0.5972 0.0409 0.3467 1>}
  n19 = VectorParameter() {group=1-Apprentice_AnimatedTriangles, parameterName=ColorBackTriangles, sortPriority=4, defaultValue=<rgba 0.2831 0.1946 0.8148 1>}
  n20 = LinearInterpolate(A=@n18.RGB, B=@n19.RGB, Alpha=Saturate(None=Ceil(None=@n12)))
  n21 = VectorParameter() {group=1-Apprentice_AnimatedTriangles, parameterName=ColorFrontTriangles, sortPriority=3, defaultValue=<rgba 0.0273 1 0.5776 1>}
  n22 = Saturate(None=Ceil(None=@n8))
  n23 = ScalarParameter() {defaultValue=0.8, group=1-Apprentice_AnimatedTriangles, parameterName=ShadedColor_Intentisty, sliderMax=1}
  n24 = NamedRerouteUsage()
  NamedRerouteDeclaration(None=Multiply(A=LinearInterpolate(A=@n20, B=@n21.RGB, Alpha=@n22), B=LinearInterpolate(A=@n23, Alpha=@n24) {constA=0.5})) {name=BC, nodeColor=<rgba 1 0.0718 0.2116 1>}
  NamedRerouteDeclaration(None=FunctionInput() {bUsePreviewValueAsDefault=true, inputName=InkLine Mask, inputType=FunctionInput_Scalar, sortPriority=2}) {name=InkLineMask, nodeColor=<rgba 0.7344 0 1 1>}
  n25 = Ceil(None=DotProduct(A=SkyAtmosphereLightDirection(), B=VertexNormalWS()))
  NamedRerouteDeclaration(None=Saturate(None=FunctionInput(Preview=@n25) {bUsePreviewValueAsDefault=true, inputName=Shading Curve, inputType=FunctionInput_Scalar, sortPriority=1})) {name=ShadingCurve, nodeColor=<rgba 1 0.9549 0 1>}
 ]}
